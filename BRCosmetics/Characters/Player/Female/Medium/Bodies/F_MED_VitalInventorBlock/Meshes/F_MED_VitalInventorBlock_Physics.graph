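# Physics asset for the F_MED_VitalInventorBlock female medium player body cosmetic, covering a secondary-motion bone chain at the right hip.
# A dyn_hipAnkle_r capsule is constrained to a tiny base_hipAnkle_r anchor capsule, with a larger thigh_r capsule that appears to act as its collider.
# The single constraint works almost as a hinge, allowing limited swing on one axis with the other swing at zero and twist nearly locked.
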PhysicsAsset /BRCosmetics/Characters/Player/Female/Medium/Bodies/F_MED_VitalInventorBlock/Meshes/F_MED_VitalInventorBlock_Physics
bodies {
 base_hipAnkle_r [{shapeType=Capsule, length=0.505, radius=0.505, shapeName=base_hipAnkle_r_capsule, center=<1 0 0>, rotation=<rot 90 0 0>}]
 dyn_hipAnkle_r [{shapeType=Capsule, length=8, radius=4, shapeName=dyn_hipAnkle_r_capsule, center=<4.9013 0 3.4527>, rotation=<rot 87.9216 179.9992 179.9993>}]
 thigh_r [{shapeType=Capsule, length=10, radius=8, shapeName=thigh_r_capsule, center=<9.9824 -0.3062 -0.1593>, rotation=<rot 90 89.4328 94.8619>}]
}
constraints [{bone1Name=dyn_hipAnkle_r, bone2Name=base_hipAnkle_r, swing1LimitDegrees=0, swing2LimitDegrees=25, twistLimitDegrees=1}]
