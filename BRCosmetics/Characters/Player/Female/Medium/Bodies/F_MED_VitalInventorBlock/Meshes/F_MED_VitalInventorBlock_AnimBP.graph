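# Post-process style AnimBlueprint for the VitalInventorBlock female medium body cosmetic, adding secondary motion to clothing and accessories.
# It copies the parent mesh pose, runs two Control Rigs, then weighted bone constraints and chained AnimDynamics for collars, scarf, zippers, cup and map.
# Additive bone offsets weighted by movement states like crouching, sprinting and kneesliding adjust accessories.
# Cached poses link the stages, ending in a pelvis-space RigidBody simulation.
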AnimBlueprint /BRCosmetics/Characters/Player/Female/Medium/Bodies/F_MED_VitalInventorBlock/Meshes/F_MED_VitalInventorBlock_AnimBP
target_skeleton=&/F_MED_VitalInventorBlock_Skeleton
graphs [
 AnimGraph {nodes [
   Output Pose(Result=Use cached pose 'CRCache_RigidBody'()) {layerGroup=None}
   n1 = Control Rig(Source=Copy Pose From Mesh() {bUseAttachedParent=true}) {
    alphaBoolBlend {blendOption=Linear}
    alphaScaleBiasClamp {bInterpResult=true}
    controlRigAssetReference {blueprintRigClass=&/F_MED_VitalInventorBlock_ControlRig.F_MED_VitalInventorBlock_ControlRig_C}
   }
   n2 = Control Rig(Source=@n1) {
    alphaBoolBlend {blendOption=Linear}
    alphaScaleBiasClamp {bInterpResult=true}
    controlRigAssetReference {blueprintRigClass=&/F_MED_VitalInventorBlockArms_CtrlRig.F_MED_VitalInventorBlockArms_CtrlRig_C}
   }
   Save cached pose 'CRCache_controlRig'(Pose=@n2) {cacheName=CRCache_controlRig}
   n3 = Anim Dynamics - Bone: dyn_collarA_c!(
    ComponentPose=Local To Component(LocalPose=Use cached pose 'CRCache_constraints'())
    Alpha=1
    GravityOverride:Vector="0,0,-980"
    ExternalForce="0,0,0"
   ) -> (Pose) {
    angularDampingOverride=1
    angularSpringConstant=20000
    bAngularSpring=true
    bOverrideAngularDamping=true
    bUseGravityOverride=true
    gravityScale=0.35
    simulationSpace=BoneRelative
    alphaBoolBlend {blendOption=Linear}
    boundBone {boneName=dyn_collarA_c}
    componentAppliedLinearAccClamp=<3000 3000 3000>
    componentLinearAccScale=<0.3 0.3 0.3>
    componentLinearVelScale=<0.3 0.3 0.3>
    gravityOverride {z=-980}
    physicsBodyDefinitions [{
      boundBone {boneName=dyn_collarA_c}
      boxExtents=<20 30 30>
      constraintSetup {angularLimitsMax=<7 7 7>, angularLimitsMin=<-7 -7 -7>, angularTarget=<1 0 0>}
      localJointOffset=<1 0 0>
     }]
    relativeSpaceBone {boneName=spine_03}
   }
   n4 = Anim Dynamics - Bone: dyn_collarB_c!(ComponentPose=@n3, Alpha=1, GravityOverride:Vector="0,0,-980", ExternalForce="0,0,0") -> (Pose) {
    angularDampingOverride=1
    angularSpringConstant=20000
    bAngularSpring=true
    bOverrideAngularDamping=true
    bUseGravityOverride=true
    gravityScale=0.35
    simulationSpace=BoneRelative
    alphaBoolBlend {blendOption=Linear}
    boundBone {boneName=dyn_collarB_c}
    componentAppliedLinearAccClamp=<3000 3000 3000>
    componentLinearAccScale=<0.3 0.3 0.3>
    componentLinearVelScale=<0.3 0.3 0.3>
    gravityOverride {z=-980}
    physicsBodyDefinitions [{
      boundBone {boneName=dyn_collarB_c}
      boxExtents=<20 30 30>
      constraintSetup {angularLimitsMax=<7 7 7>, angularLimitsMin=<-7 -7 -7>, angularTarget=<1 0 0>}
      localJointOffset=<1 0 0>
     }]
    relativeSpaceBone {boneName=spine_03}
   }
   n5 = Anim Dynamics - Bone: dyn_collar1A_r!(ComponentPose=@n4, Alpha=1, GravityOverride:Vector="0,0,-980", ExternalForce="0,0,0") -> (Pose) {
    angularDampingOverride=1
    angularSpringConstant=20000
    bAngularSpring=true
    bOverrideAngularDamping=true
    bUseGravityOverride=true
    gravityScale=0.35
    simulationSpace=BoneRelative
    alphaBoolBlend {blendOption=Linear}
    boundBone {boneName=dyn_collar1A_r}
    componentAppliedLinearAccClamp=<3000 3000 3000>
    componentLinearAccScale=<0.3 0.3 0.3>
    componentLinearVelScale=<0.3 0.3 0.3>
    gravityOverride {z=-980}
    physicsBodyDefinitions [{
      boundBone {boneName=dyn_collar1A_r}
      boxExtents=<20 30 30>
      constraintSetup {angularLimitsMax=<7 7 7>, angularLimitsMin=<-7 -7 -7>, angularTarget=<1 0 0>}
      localJointOffset=<1 0 0>
     }]
    relativeSpaceBone {boneName=spine_03}
   }
   n6 = Anim Dynamics - Bone: dyn_collar1B_r!(ComponentPose=@n5, Alpha=1, GravityOverride:Vector="0,0,-980", ExternalForce="0,0,0") -> (Pose) {
    angularDampingOverride=1
    angularSpringConstant=20000
    bAngularSpring=true
    bOverrideAngularDamping=true
    bUseGravityOverride=true
    gravityScale=0.35
    simulationSpace=BoneRelative
    alphaBoolBlend {blendOption=Linear}
    boundBone {boneName=dyn_collar1B_r}
    componentAppliedLinearAccClamp=<3000 3000 3000>
    componentLinearAccScale=<0.3 0.3 0.3>
    componentLinearVelScale=<0.3 0.3 0.3>
    gravityOverride {z=-980}
    physicsBodyDefinitions [{
      boundBone {boneName=dyn_collar1B_r}
      boxExtents=<20 30 30>
      constraintSetup {angularLimitsMax=<7 7 7>, angularLimitsMin=<-7 -7 -7>, angularTarget=<1 0 0>}
      localJointOffset=<1 0 0>
     }]
    relativeSpaceBone {boneName=spine_03}
   }
   n7 = Anim Dynamics - Bone: dyn_collar1A_l!(ComponentPose=@n6, Alpha=1, GravityOverride:Vector="0,0,-980", ExternalForce="0,0,0") -> (Pose) {
    angularDampingOverride=1
    angularSpringConstant=20000
    bAngularSpring=true
    bOverrideAngularDamping=true
    bUseGravityOverride=true
    gravityScale=0.35
    simulationSpace=BoneRelative
    alphaBoolBlend {blendOption=Linear}
    boundBone {boneName=dyn_collar1A_l}
    componentAppliedLinearAccClamp=<3000 3000 3000>
    componentLinearAccScale=<0.3 0.3 0.3>
    componentLinearVelScale=<0.3 0.3 0.3>
    gravityOverride {z=-980}
    physicsBodyDefinitions [{
      boundBone {boneName=dyn_collar1A_l}
      boxExtents=<20 30 30>
      constraintSetup {angularLimitsMax=<7 7 7>, angularLimitsMin=<-7 -7 -7>, angularTarget=<1 0 0>}
      localJointOffset=<1 0 0>
     }]
    relativeSpaceBone {boneName=spine_03}
   }
   n8 = Anim Dynamics - Bone: dyn_collar1B_l!(ComponentPose=@n7, Alpha=1, GravityOverride:Vector="0,0,-980", ExternalForce="0,0,0") -> (Pose) {
    angularDampingOverride=1
    angularSpringConstant=20000
    bAngularSpring=true
    bOverrideAngularDamping=true
    bUseGravityOverride=true
    gravityScale=0.35
    simulationSpace=BoneRelative
    alphaBoolBlend {blendOption=Linear}
    boundBone {boneName=dyn_collar1B_l}
    componentAppliedLinearAccClamp=<3000 3000 3000>
    componentLinearAccScale=<0.3 0.3 0.3>
    componentLinearVelScale=<0.3 0.3 0.3>
    gravityOverride {z=-980}
    physicsBodyDefinitions [{
      boundBone {boneName=dyn_collar1B_l}
      boxExtents=<20 30 30>
      constraintSetup {angularLimitsMax=<7 7 7>, angularLimitsMin=<-7 -7 -7>, angularTarget=<1 0 0>}
      localJointOffset=<1 0 0>
     }]
    relativeSpaceBone {boneName=spine_03}
   }
   n9 = Anim Dynamics - Bone: dyn_collar2A_l!(ComponentPose=@n8, Alpha=1, GravityOverride:Vector="0,0,-980", ExternalForce="0,0,0") -> (Pose) {
    angularDampingOverride=1
    angularSpringConstant=20000
    bAngularSpring=true
    bOverrideAngularDamping=true
    bUseGravityOverride=true
    gravityScale=0.35
    simulationSpace=BoneRelative
    alphaBoolBlend {blendOption=Linear}
    boundBone {boneName=dyn_collar2A_l}
    componentAppliedLinearAccClamp=<3000 3000 3000>
    componentLinearAccScale=<0.3 0.3 0.3>
    componentLinearVelScale=<0.3 0.3 0.3>
    gravityOverride {z=-980}
    physicsBodyDefinitions [{
      boundBone {boneName=dyn_collar2A_l}
      boxExtents=<20 30 30>
      constraintSetup {angularLimitsMax=<7 7 7>, angularLimitsMin=<-7 -7 -7>, angularTarget=<1 0 0>}
      localJointOffset=<1 0 0>
     }]
    relativeSpaceBone {boneName=spine_03}
   }
   n10 = Anim Dynamics - Bone: dyn_collar2B_l!(ComponentPose=@n9, Alpha=1, GravityOverride:Vector="0,0,-980", ExternalForce="0,0,0") -> (Pose) {
    angularDampingOverride=1
    angularSpringConstant=20000
    bAngularSpring=true
    bOverrideAngularDamping=true
    bUseGravityOverride=true
    gravityScale=0.35
    simulationSpace=BoneRelative
    alphaBoolBlend {blendOption=Linear}
    boundBone {boneName=dyn_collar2B_l}
    componentAppliedLinearAccClamp=<3000 3000 3000>
    componentLinearAccScale=<0.3 0.3 0.3>
    componentLinearVelScale=<0.3 0.3 0.3>
    gravityOverride {z=-980}
    physicsBodyDefinitions [{
      boundBone {boneName=dyn_collar2B_l}
      boxExtents=<20 30 30>
      constraintSetup {angularLimitsMax=<7 7 7>, angularLimitsMin=<-7 -7 -7>, angularTarget=<1 0 0>}
      localJointOffset=<1 0 0>
     }]
    relativeSpaceBone {boneName=spine_03}
   }
   n11 = Anim Dynamics - Bone: dyn_collar2A_r!(ComponentPose=@n10, Alpha=1, GravityOverride:Vector="0,0,-980", ExternalForce="0,0,0") -> (Pose) {
    angularDampingOverride=1
    angularSpringConstant=20000
    bAngularSpring=true
    bOverrideAngularDamping=true
    bUseGravityOverride=true
    gravityScale=0.35
    simulationSpace=BoneRelative
    alphaBoolBlend {blendOption=Linear}
    boundBone {boneName=dyn_collar2A_r}
    componentAppliedLinearAccClamp=<3000 3000 3000>
    componentLinearAccScale=<0.3 0.3 0.3>
    componentLinearVelScale=<0.3 0.3 0.3>
    gravityOverride {z=-980}
    physicsBodyDefinitions [{
      boundBone {boneName=dyn_collar2A_r}
      boxExtents=<20 30 30>
      constraintSetup {angularLimitsMax=<7 7 7>, angularLimitsMin=<-7 -7 -7>, angularTarget=<1 0 0>}
      localJointOffset=<1 0 0>
     }]
    relativeSpaceBone {boneName=spine_03}
   }
   n12 = Anim Dynamics - Bone: dyn_collar2B_r!(ComponentPose=@n11, Alpha=1, GravityOverride:Vector="0,0,-980", ExternalForce="0,0,0") -> (Pose) {
    angularDampingOverride=1
    angularSpringConstant=20000
    bAngularSpring=true
    bOverrideAngularDamping=true
    bUseGravityOverride=true
    gravityScale=0.35
    simulationSpace=BoneRelative
    alphaBoolBlend {blendOption=Linear}
    boundBone {boneName=dyn_collar2B_r}
    componentAppliedLinearAccClamp=<3000 3000 3000>
    componentLinearAccScale=<0.3 0.3 0.3>
    componentLinearVelScale=<0.3 0.3 0.3>
    gravityOverride {z=-980}
    physicsBodyDefinitions [{
      boundBone {boneName=dyn_collar2B_r}
      boxExtents=<20 30 30>
      constraintSetup {angularLimitsMax=<7 7 7>, angularLimitsMin=<-7 -7 -7>, angularTarget=<1 0 0>}
      localJointOffset=<1 0 0>
     }]
    relativeSpaceBone {boneName=spine_03}
   }
   n13 = Anim Dynamics - Bone: dyn_boatA_l(ComponentPose=@n12, Alpha=1, ExternalForce="0,0,0") {
    angularDampingOverride=1
    angularSpringConstant=20000
    bAngularSpring=true
    bOverrideAngularDamping=true
    bUseGravityOverride=true
    gravityScale=0.35
    simulationSpace=BoneRelative
    alphaBoolBlend {blendOption=Linear}
    alphaScaleBiasClamp {bInterpResult=true}
    boundBone {boneName=dyn_boatA_l}
    componentLinearAccScale=<1 1 1>
    componentLinearVelScale=<1 1 1>
    gravityOverride {z=-980}
    physicsBodyDefinitions [{
      boundBone {boneName=dyn_boatA_l}
      boxExtents=<20 30 30>
      constraintSetup {angularLimitsMax=<15 15 15>, angularLimitsMin=<-15 -15 -15>, angularTarget=<1 0 0>}
      localJointOffset=<4 0 0>
     }]
    relativeSpaceBone {boneName=thigh_l}
   }
   n14 = Anim Dynamics - Bone: dyn_boatC_l(ComponentPose=@n13, Alpha=1, ExternalForce="0,0,0") {
    angularDampingOverride=1
    angularSpringConstant=20000
    bAngularSpring=true
    bOverrideAngularDamping=true
    bUseGravityOverride=true
    gravityScale=0.35
    simulationSpace=BoneRelative
    alphaBoolBlend {blendOption=Linear}
    alphaScaleBiasClamp {bInterpResult=true}
    boundBone {boneName=dyn_boatC_l}
    componentLinearAccScale=<1 1 1>
    componentLinearVelScale=<1 1 1>
    gravityOverride {z=-980}
    physicsBodyDefinitions [{
      boundBone {boneName=dyn_boatC_l}
      boxExtents=<20 30 30>
      constraintSetup {angularLimitsMax=<15 15 15>, angularLimitsMin=<-15 -15 -15>, angularTarget=<1 0 0>}
      localJointOffset=<4 0 0>
     }]
    relativeSpaceBone {boneName=thigh_l}
   }
   n15 = Anim Dynamics - Bone: dyn_boatB_l(ComponentPose=@n14, Alpha=1, ExternalForce="0,0,0") {
    angularDampingOverride=1
    angularSpringConstant=20000
    bAngularSpring=true
    bOverrideAngularDamping=true
    bUseGravityOverride=true
    gravityScale=0.35
    simulationSpace=BoneRelative
    alphaBoolBlend {blendOption=Linear}
    alphaScaleBiasClamp {bInterpResult=true}
    boundBone {boneName=dyn_boatB_l}
    componentLinearAccScale=<1 1 1>
    componentLinearVelScale=<1 1 1>
    gravityOverride {z=-980}
    physicsBodyDefinitions [{
      boundBone {boneName=dyn_boatB_l}
      boxExtents=<20 30 30>
      constraintSetup {angularLimitsMax=<15 15 15>, angularLimitsMin=<-15 -15 -15>, angularTarget=<1 0 0>}
      localJointOffset=<4 0 0>
     }]
    relativeSpaceBone {boneName=thigh_l}
   }
   n16 = Anim Dynamics - Bone: dyn_zipper_l(ComponentPose=@n15, Alpha=1, ExternalForce="0,0,0") {
    angularDampingOverride=1
    angularSpringConstant=20000
    bAngularSpring=true
    bOverrideAngularDamping=true
    bUseGravityOverride=true
    gravityScale=0.35
    simulationSpace=BoneRelative
    alphaBoolBlend {blendOption=Linear}
    alphaScaleBiasClamp {bInterpResult=true}
    boundBone {boneName=dyn_zipper_l}
    componentLinearAccScale=<1 1 1>
    componentLinearVelScale=<1 1 1>
    gravityOverride {z=-980}
    physicsBodyDefinitions [{
      boundBone {boneName=dyn_zipper_l}
      boxExtents=<20 30 30>
      constraintSetup {angularLimitsMax=<15 15 15>, angularLimitsMin=<-15 -15 -15>, angularTarget=<1 0 0>}
      localJointOffset=<4 0 0>
     }]
    relativeSpaceBone {boneName=thigh_l}
   }
   n17 = Anim Dynamics - Bone: dyn_boatB_r(ComponentPose=@n16, Alpha=1, ExternalForce="0,0,0") {
    angularDampingOverride=1
    angularSpringConstant=20000
    bAngularSpring=true
    bOverrideAngularDamping=true
    bUseGravityOverride=true
    gravityScale=0.35
    simulationSpace=BoneRelative
    alphaBoolBlend {blendOption=Linear}
    alphaScaleBiasClamp {bInterpResult=true}
    boundBone {boneName=dyn_boatB_r}
    componentLinearAccScale=<1 1 1>
    componentLinearVelScale=<1 1 1>
    gravityOverride {z=-980}
    physicsBodyDefinitions [{
      boundBone {boneName=dyn_boatB_r}
      boxExtents=<20 30 30>
      constraintSetup {angularLimitsMax=<15 15 15>, angularLimitsMin=<-15 -15 -15>, angularTarget=<1 0 0>}
      localJointOffset=<4 0 0>
     }]
    relativeSpaceBone {boneName=thigh_r}
   }
   n18 = Anim Dynamics - Bone: dyn_boatC_r(ComponentPose=@n17, Alpha=1, ExternalForce="0,0,0") {
    angularDampingOverride=1
    angularSpringConstant=20000
    bAngularSpring=true
    bOverrideAngularDamping=true
    bUseGravityOverride=true
    gravityScale=0.35
    simulationSpace=BoneRelative
    alphaBoolBlend {blendOption=Linear}
    alphaScaleBiasClamp {bInterpResult=true}
    boundBone {boneName=dyn_boatC_r}
    componentLinearAccScale=<1 1 1>
    componentLinearVelScale=<1 1 1>
    gravityOverride {z=-980}
    physicsBodyDefinitions [{
      boundBone {boneName=dyn_boatC_r}
      boxExtents=<20 30 30>
      constraintSetup {angularLimitsMax=<15 15 15>, angularLimitsMin=<-15 -15 -15>, angularTarget=<1 0 0>}
      localJointOffset=<4 0 0>
     }]
    relativeSpaceBone {boneName=thigh_r}
   }
   n19 = Anim Dynamics - Bone: dyn_boatA_r(ComponentPose=@n18, Alpha=1, ExternalForce="0,0,0") {
    angularDampingOverride=1
    angularSpringConstant=20000
    bAngularSpring=true
    bOverrideAngularDamping=true
    bUseGravityOverride=true
    gravityScale=0.35
    simulationSpace=BoneRelative
    alphaBoolBlend {blendOption=Linear}
    alphaScaleBiasClamp {bInterpResult=true}
    boundBone {boneName=dyn_boatA_r}
    componentLinearAccScale=<1 1 1>
    componentLinearVelScale=<1 1 1>
    gravityOverride {z=-980}
    physicsBodyDefinitions [{
      boundBone {boneName=dyn_boatA_r}
      boxExtents=<20 30 30>
      constraintSetup {angularLimitsMax=<15 15 15>, angularLimitsMin=<-15 -15 -15>, angularTarget=<1 0 0>}
      localJointOffset=<4 0 0>
     }]
    relativeSpaceBone {boneName=thigh_r}
   }
   n20 = Anim Dynamics - Bone: dyn_zipper_r(ComponentPose=@n19, Alpha=1, ExternalForce="0,0,0") {
    angularDampingOverride=1
    angularSpringConstant=20000
    bAngularSpring=true
    bOverrideAngularDamping=true
    bUseGravityOverride=true
    gravityScale=0.35
    simulationSpace=BoneRelative
    alphaBoolBlend {blendOption=Linear}
    alphaScaleBiasClamp {bInterpResult=true}
    boundBone {boneName=dyn_zipper_r}
    componentLinearAccScale=<1 1 1>
    componentLinearVelScale=<1 1 1>
    gravityOverride {z=-980}
    physicsBodyDefinitions [{
      boundBone {boneName=dyn_zipper_r}
      boxExtents=<20 30 30>
      constraintSetup {angularLimitsMax=<15 15 15>, angularLimitsMin=<-15 -15 -15>, angularTarget=<1 0 0>}
      localJointOffset=<4 0 0>
     }]
    relativeSpaceBone {boneName=thigh_r}
   }
   n21 = Anim Dynamics - Bone: dyn_scarfB_03(ComponentPose=@n20, Alpha=1, ExternalForce="0,0,0") {
    angularDampingOverride=1
    angularSpringConstant=20000
    bAngularSpring=true
    bOverrideAngularDamping=true
    bUseGravityOverride=true
    gravityScale=0.35
    simulationSpace=BoneRelative
    alphaBoolBlend {blendOption=Linear}
    alphaScaleBiasClamp {bInterpResult=true, interpSpeedDecreasing=6, interpSpeedIncreasing=4}
    boundBone {boneName=dyn_scarfB_03}
    componentAppliedLinearAccClamp=<3000 3000 3000>
    componentLinearAccScale=<0.3 0.3 0.3>
    componentLinearVelScale=<0.3 0.3 0.3>
    gravityOverride {z=-980}
    physicsBodyDefinitions [{
      boundBone {boneName=dyn_scarfB_03}
      boxExtents=<20 30 30>
      constraintSetup {angularLimitsMax=<25 25 25>, angularLimitsMin=<-25 -25 -25>, angularTarget=<1 0 0>}
      localJointOffset=<4 0 0>
     }]
    relativeSpaceBone {boneName=upperarm_l}
   }
   n22 = Anim Dynamics - Bone: dyn_scarfA_04(ComponentPose=@n21, Alpha=1, ExternalForce="0,0,0") {
    angularDampingOverride=1
    angularSpringConstant=20000
    bAngularSpring=true
    bOverrideAngularDamping=true
    bUseGravityOverride=true
    gravityScale=0.35
    simulationSpace=BoneRelative
    alphaBoolBlend {blendOption=Linear}
    alphaScaleBiasClamp {bInterpResult=true, interpSpeedDecreasing=6, interpSpeedIncreasing=4}
    boundBone {boneName=dyn_scarfA_04}
    componentAppliedLinearAccClamp=<3000 3000 3000>
    componentLinearAccScale=<0.3 0.3 0.3>
    componentLinearVelScale=<0.3 0.3 0.3>
    gravityOverride {z=-980}
    physicsBodyDefinitions [{
      boundBone {boneName=dyn_scarfA_04}
      boxExtents=<20 30 30>
      constraintSetup {angularLimitsMax=<25 25 25>, angularLimitsMin=<-25 -25 -25>, angularTarget=<1 0 0>}
      localJointOffset=<4 0 0>
     }]
    relativeSpaceBone {boneName=upperarm_l}
   }
   n23 = Anim Dynamics - Bone: dyn_scarfA_03(ComponentPose=@n22, Alpha=1, ExternalForce="0,0,0") {
    angularDampingOverride=1
    angularSpringConstant=20000
    bAngularSpring=true
    bOverrideAngularDamping=true
    bUseGravityOverride=true
    gravityScale=0.35
    simulationSpace=BoneRelative
    alphaBoolBlend {blendOption=Linear}
    alphaScaleBiasClamp {bInterpResult=true, interpSpeedDecreasing=6, interpSpeedIncreasing=4}
    boundBone {boneName=dyn_scarfA_03}
    componentAppliedLinearAccClamp=<3000 3000 3000>
    componentLinearAccScale=<0.3 0.3 0.3>
    componentLinearVelScale=<0.3 0.3 0.3>
    gravityOverride {z=-980}
    physicsBodyDefinitions [{
      boundBone {boneName=dyn_scarfA_03}
      boxExtents=<20 30 30>
      constraintSetup {angularLimitsMax=<25 25 25>, angularLimitsMin=<-25 -25 -25>, angularTarget=<1 0 0>}
      localJointOffset=<4 0 0>
     }]
    relativeSpaceBone {boneName=upperarm_l}
   }
   n24 = Anim Dynamics - Bone: dyn_scarfB_02(ComponentPose=@n23, Alpha=1, ExternalForce="0,0,0") {
    angularDampingOverride=1
    angularSpringConstant=20000
    bAngularSpring=true
    bOverrideAngularDamping=true
    bUseGravityOverride=true
    gravityScale=0.35
    simulationSpace=BoneRelative
    alphaBoolBlend {blendOption=Linear}
    alphaScaleBiasClamp {bInterpResult=true, interpSpeedDecreasing=6, interpSpeedIncreasing=4}
    boundBone {boneName=dyn_scarfB_02}
    componentAppliedLinearAccClamp=<3000 3000 3000>
    componentLinearAccScale=<0.3 0.3 0.3>
    componentLinearVelScale=<0.3 0.3 0.3>
    gravityOverride {z=-980}
    physicsBodyDefinitions [{
      boundBone {boneName=dyn_scarfB_02}
      boxExtents=<20 30 30>
      constraintSetup {angularLimitsMax=<25 25 25>, angularLimitsMin=<-25 -25 -25>, angularTarget=<1 0 0>}
      localJointOffset=<4 0 0>
     }]
    relativeSpaceBone {boneName=upperarm_l}
   }
   n25 = Anim Dynamics - Bone: dyn_scarfA_02(ComponentPose=@n24, Alpha=1, ExternalForce="0,0,0") {
    angularDampingOverride=1
    angularSpringConstant=20000
    bAngularSpring=true
    bOverrideAngularDamping=true
    bUseGravityOverride=true
    gravityScale=0.35
    simulationSpace=BoneRelative
    alphaBoolBlend {blendOption=Linear}
    alphaScaleBiasClamp {bInterpResult=true, interpSpeedDecreasing=6, interpSpeedIncreasing=4}
    boundBone {boneName=dyn_scarfA_02}
    componentAppliedLinearAccClamp=<3000 3000 3000>
    componentLinearAccScale=<0.3 0.3 0.3>
    componentLinearVelScale=<0.3 0.3 0.3>
    gravityOverride {z=-980}
    physicsBodyDefinitions [{
      boundBone {boneName=dyn_scarfA_02}
      boxExtents=<20 30 30>
      constraintSetup {angularLimitsMax=<25 25 25>, angularLimitsMin=<-25 -25 -25>, angularTarget=<1 0 0>}
      localJointOffset=<4 0 0>
     }]
    relativeSpaceBone {boneName=upperarm_l}
   }
   n26 = Anim Dynamics - Bone: dyn_scarfA_01(ComponentPose=@n25, Alpha=1, ExternalForce="0,0,0") {
    angularDampingOverride=1
    angularSpringConstant=20000
    bAngularSpring=true
    bOverrideAngularDamping=true
    bUseGravityOverride=true
    gravityScale=0.35
    simulationSpace=BoneRelative
    alphaBoolBlend {blendOption=Linear}
    alphaScaleBiasClamp {bInterpResult=true, interpSpeedDecreasing=6, interpSpeedIncreasing=4}
    boundBone {boneName=dyn_scarfA_01}
    componentAppliedLinearAccClamp=<3000 3000 3000>
    componentLinearAccScale=<0.3 0.3 0.3>
    componentLinearVelScale=<0.3 0.3 0.3>
    gravityOverride {z=-980}
    physicsBodyDefinitions [{
      boundBone {boneName=dyn_scarfA_01}
      boxExtents=<20 30 30>
      constraintSetup {angularLimitsMax=<25 25 25>, angularLimitsMin=<-25 -25 -25>, angularTarget=<1 0 0>}
      localJointOffset=<4 0 0>
     }]
    relativeSpaceBone {boneName=upperarm_l}
   }
   n27 = Anim Dynamics - Bone: dyn_scarfB_01(ComponentPose=@n26, Alpha=1, ExternalForce="0,0,0") {
    angularDampingOverride=1
    angularSpringConstant=20000
    bAngularSpring=true
    bOverrideAngularDamping=true
    bUseGravityOverride=true
    gravityScale=0.35
    simulationSpace=BoneRelative
    alphaBoolBlend {blendOption=Linear}
    alphaScaleBiasClamp {bInterpResult=true, interpSpeedDecreasing=6, interpSpeedIncreasing=4}
    boundBone {boneName=dyn_scarfB_01}
    componentAppliedLinearAccClamp=<3000 3000 3000>
    componentLinearAccScale=<0.3 0.3 0.3>
    componentLinearVelScale=<0.3 0.3 0.3>
    gravityOverride {z=-980}
    physicsBodyDefinitions [{
      boundBone {boneName=dyn_scarfB_01}
      boxExtents=<20 30 30>
      constraintSetup {angularLimitsMax=<25 25 25>, angularLimitsMin=<-25 -25 -25>, angularTarget=<1 0 0>}
      localJointOffset=<4 0 0>
     }]
    relativeSpaceBone {boneName=upperarm_l}
   }
   n28 = Anim Dynamics - Bone: dyn_cup_r(ComponentPose=@n27, Alpha=1, ExternalForce="0,0,0") {
    angularDampingOverride=1
    angularSpringConstant=20000
    bAngularSpring=true
    bOverrideAngularDamping=true
    bUseGravityOverride=true
    gravityScale=0.35
    simulationSpace=BoneRelative
    alphaBoolBlend {blendOption=Linear}
    alphaScaleBiasClamp {bInterpResult=true}
    boundBone {boneName=dyn_cup_r}
    componentAppliedLinearAccClamp=<10000 10000 10000>
    componentLinearAccScale=<1 1 1>
    componentLinearVelScale=<1 1 1>
    gravityOverride {z=-980}
    physicsBodyDefinitions [{
      boundBone {boneName=dyn_cup_r}
      boxExtents=<20 30 30>
      constraintSetup {angularLimitsMax=<80 80 0>, angularLimitsMin=<-60 -80 -30>, angularTarget=<1 0 0>}
      localJointOffset=<3 0 0>
     }]
    relativeSpaceBone {boneName=spine_01}
   }
   n29 = Anim Dynamics - Bone: dyn_zipper_l(ComponentPose=@n28, Alpha=1, ExternalForce="0,0,0") {
    angularDampingOverride=1
    angularSpringConstant=20000
    bAngularSpring=true
    bOverrideAngularDamping=true
    bUseGravityOverride=true
    gravityScale=0.35
    simulationSpace=BoneRelative
    alphaBoolBlend {blendOption=Linear}
    alphaScaleBiasClamp {bInterpResult=true}
    boundBone {boneName=dyn_zipper_l}
    componentLinearAccScale=<1 1 1>
    componentLinearVelScale=<1 1 1>
    gravityOverride {z=-980}
    physicsBodyDefinitions [{
      boundBone {boneName=dyn_zipper_l}
      boxExtents=<20 30 30>
      constraintSetup {angularLimitsMax=<15 15 15>, angularLimitsMin=<-15 -15 -15>, angularTarget=<1 0 0>}
      localJointOffset=<4 0 0>
     }]
    relativeSpaceBone {boneName=calf_l}
   }
   n30 = Anim Dynamics - Bone: dyn_zipper_r(ComponentPose=@n29, Alpha=1, ExternalForce="0,0,0") {
    angularDampingOverride=1
    angularSpringConstant=20000
    bAngularSpring=true
    bOverrideAngularDamping=true
    bUseGravityOverride=true
    gravityScale=0.35
    simulationSpace=BoneRelative
    alphaBoolBlend {blendOption=Linear}
    alphaScaleBiasClamp {bInterpResult=true}
    boundBone {boneName=dyn_zipper_r}
    componentLinearAccScale=<1 1 1>
    componentLinearVelScale=<1 1 1>
    gravityOverride {z=-980}
    physicsBodyDefinitions [{
      boundBone {boneName=dyn_zipper_r}
      boxExtents=<20 30 30>
      constraintSetup {angularLimitsMax=<15 15 15>, angularLimitsMin=<-15 -15 -15>, angularTarget=<1 0 0>}
      localJointOffset=<4 0 0>
     }]
    relativeSpaceBone {boneName=calf_r}
   }
   n31 = Anim Dynamics - Bone: dyn_mapB_c!(ComponentPose=@n30, Alpha=1, GravityOverride:Vector="0,0,0", ExternalForce="0,0,0") -> (Pose) {
    angularDampingOverride=1
    angularSpringConstant=20000
    bAngularSpring=true
    bOverrideAngularDamping=true
    gravityScale=0.35
    simulationSpace=BoneRelative
    alphaBoolBlend {blendOption=Linear}
    alphaScaleBiasClamp {bInterpResult=true, interpSpeedDecreasing=20}
    boundBone {boneName=dyn_mapB_c}
    componentAppliedLinearAccClamp=<3000 3000 3000>
    componentLinearAccScale=<1 1 1>
    componentLinearVelScale=<1 1 1>
    gravityOverride {z=-980}
    physicsBodyDefinitions [{
      boundBone {boneName=dyn_mapB_c}
      boxExtents=<20 30 30>
      constraintSetup {angularLimitsMax=<0 0 0>, angularLimitsMin=<0 0 -10>, angularTarget=<0 0 0>}
      localJointOffset=<2 0 0>
     }]
    relativeSpaceBone {boneName=root}
   }
   Save cached pose 'CRCache_Dynamics'(Pose=Component To Local(ComponentPose=@n31)) {cacheName=CRCache_Dynamics}
   n32 = "Constraint - base_dyn_cup_r (PPPPP)"!(
    ComponentPose=Local To Component(LocalPose=Use cached pose 'CRCache_controlRig'())
    Alpha=1
    ConstraintWeights_0=0.5856
    ConstraintWeights_1:float=0.2521
    ConstraintWeights_2:float=0.162
    ConstraintWeights_3:float=0.0002
    ConstraintWeights_4:float=0.0001
   ) -> (Pose) {
    alphaBoolBlend {blendOption=Linear}
    boneToModify {boneName=base_dyn_cup_r}
    constraintSetup [
     {targetBone {boneName=pelvis}}
     {targetBone {boneName=deform_glute_r}}
     {targetBone {boneName=thigh_twist_01_r}}
     {targetBone {boneName=deform_groin_r}}
     {targetBone {boneName=thigh_r}}
    ]
    constraintWeights [1,1,1,1,1]
   }
   n33 = "Constraint - base_bag_r (PPPPPPPP)"!(
    ComponentPose=@n32
    Alpha=1
    ConstraintWeights_0=0.6226
    ConstraintWeights_1:float=0.2776
    ConstraintWeights_2:float=0.0479
    ConstraintWeights_3:float=0.0326
    ConstraintWeights_4:float=0.0181
    ConstraintWeights_5:float=0.0007
    ConstraintWeights_6:float=0.0003
    ConstraintWeights_7:float=0.0002
   ) -> (Pose) {
    alphaBoolBlend {blendOption=Linear}
    boneToModify {boneName=base_bag_r}
    constraintSetup [
     {targetBone {boneName=pelvis}}
     {targetBone {boneName=thigh_twist_01_r}}
     {targetBone {boneName=deform_groin_r}}
     {targetBone {boneName=deform_glute_r}}
     {targetBone {boneName=thigh_r}}
     {targetBone {boneName=spine_01}}
     {targetBone {boneName=thigh_twist_01_l}}
     {targetBone {boneName=lowerarm_twist_02_r}}
    ]
    constraintWeights [1,1,1,1,1,1,1,1]
   }
   n34 = "Constraint - base_map_c (PPPPPP)"!(
    ComponentPose=@n33
    Alpha=1
    ConstraintWeights_0=0.8888
    ConstraintWeights_1:float=0.0396
    ConstraintWeights_2:float=0.0345
    ConstraintWeights_3:float=0.0333
    ConstraintWeights_4:float=0.0019
    ConstraintWeights_5:float=0.0019
   ) -> (Pose) {
    alphaBoolBlend {blendOption=Linear}
    alphaScaleBiasClamp {bInterpResult=true}
    boneToModify {boneName=base_map_c}
    constraintSetup [
     {targetBone {boneName=pelvis}}
     {targetBone {boneName=deform_glute_r}}
     {targetBone {boneName=spine_01}}
     {targetBone {boneName=deform_glute_l}}
     {targetBone {boneName=thigh_twist_01_r}}
     {targetBone {boneName=thigh_twist_01_l}}
    ]
    constraintWeights [1,1,1,1,1,1]
   }
   n35 = "Constraint - base_hipAnkle_r (PPPPPPP)"!(
    ComponentPose=@n34
    Alpha=1
    ConstraintWeights_0=0.5773
    ConstraintWeights_1:float=0.2689
    ConstraintWeights_2:float=0.1128
    ConstraintWeights_3:float=0.0318
    ConstraintWeights_4:float=0.0047
    ConstraintWeights_5:float=0.0033
    ConstraintWeights_6:float=0.0012
   ) -> (Pose) {
    alphaBoolBlend {blendOption=Linear}
    boneToModify {boneName=base_hipAnkle_r}
    constraintSetup [
     {targetBone {boneName=pelvis}}
     {targetBone {boneName=thigh_twist_01_r}}
     {targetBone {boneName=deform_groin_r}}
     {targetBone {boneName=thigh_r}}
     {targetBone {boneName=deform_glute_r}}
     {targetBone {boneName=thigh_twist_01_l}}
     {targetBone {boneName=deform_groin_l}}
    ]
    constraintWeights [1,1,1,1,1,1,1]
   }
   Save cached pose 'CRCache_constraints'(Pose=Component To Local(ComponentPose=@n35)) {cacheName=CRCache_constraints}
   n36 = "Transform (Modify) Bone - Bone: base_map_c"(
    ComponentPose=Local To Component(LocalPose=Use cached pose 'CRCache_Dynamics'())
    Alpha=Get KneeslidingAlpha() -> (KneeslidingAlpha:float)
    Translation="-0.938,1.574,3.035"
    Rotation="-50.551,6.786,26.763"
    Scale="1,1,1"
   ) {
    rotationMode=BMM_Additive
    rotationSpace=BCS_BoneSpace
    translationMode=BMM_Additive
    translationSpace=BCS_BoneSpace
    alphaBoolBlend {blendOption=Linear}
    alphaScaleBiasClamp {bInterpResult=true}
    boneToModify {boneName=base_map_c}
    rotation=<rot -50.551 6.786 26.763>
    translation=<-0.938 1.574 3.035>
   }
   Save cached pose 'CRCache_map'(Pose=Component To Local(ComponentPose=@n36)) {cacheName=CRCache_map}
   n37 = "Transform (Modify) Bone - Bone: base_dyn_cup_r"(
    ComponentPose=Local To Component(LocalPose=Use cached pose 'CRCache_map'())
    Alpha=Get crouching() -> (crouching:double)
    Translation="0,-1,1"
    Rotation="0,0,0"
    Scale="1,1,1"
   ) {
    rotationMode=BMM_Additive
    rotationSpace=BCS_BoneSpace
    translationMode=BMM_Additive
    translationSpace=BCS_BoneSpace
    alphaBoolBlend {blendOption=Linear}
    alphaScaleBiasClamp {bInterpResult=true}
    boneToModify {boneName=base_dyn_cup_r}
    translation {y=-1, z=1}
   }
   n38 = "Transform (Modify) Bone - Bone: base_dyn_cup_r"(ComponentPose=@n37, Alpha=Get jumping() -> (jumping:double), Translation="0,-0.75,0.25", Rotation="0,0,0", Scale="1,1,1") {
    rotationMode=BMM_Additive
    rotationSpace=BCS_BoneSpace
    translationMode=BMM_Additive
    translationSpace=BCS_BoneSpace
    alphaBoolBlend {blendOption=Linear}
    alphaScaleBiasClamp {bInterpResult=true}
    boneToModify {boneName=base_dyn_cup_r}
    translation {y=-0.75, z=0.25}
   }
   n39 = "Transform (Modify) Bone - Bone: base_dyn_cup_r"(ComponentPose=@n38, Alpha=Get sprinting() -> (sprinting:double), Translation="0,-1,1", Rotation="0,0,0", Scale="1,1,1") {
    rotationMode=BMM_Additive
    rotationSpace=BCS_BoneSpace
    translationMode=BMM_Additive
    translationSpace=BCS_BoneSpace
    alphaBoolBlend {blendOption=Linear}
    alphaScaleBiasClamp {bInterpResult=true}
    boneToModify {boneName=base_dyn_cup_r}
    translation {y=-1, z=1}
   }
   n40 = "Transform (Modify) Bone - Bone: base_dyn_cup_r"(ComponentPose=@n39, Alpha=Get KneeslidingAlpha(), Translation="-0.594,1.795,-2.780", Rotation="-28.949,-90.146,29.021", Scale="1,1,1") {
    rotationMode=BMM_Additive
    rotationSpace=BCS_BoneSpace
    translationMode=BMM_Additive
    translationSpace=BCS_BoneSpace
    alphaBoolBlend {blendOption=Linear}
    alphaScaleBiasClamp {bInterpResult=true}
    boneToModify {boneName=base_dyn_cup_r}
    rotation=<rot -28.949 -90.146 29.021>
    translation=<-0.594 1.795 -2.78>
   }
   Save cached pose 'CRCache_Cup'(Pose=Component To Local(ComponentPose=@n40)) {cacheName=CRCache_Cup}
   n41 = RigidBody!(
    ComponentPose=Local To Component(LocalPose=Use cached pose 'CRCache_Neck'())
    Alpha
    OverrideWorldGravity:Vector=Get Override World Gravity() -> (Override World Gravity:Vector)
    ExternalForce
   ) -> (Pose) {
    bOverrideWorldGravity=true
    overridePhysicsAsset=&/F_MED_VitalInventorBlock_Physics
    simulationSpace=BaseBoneSpace
    alphaBoolBlend {blendOption=Linear}
    baseBoneRef {boneName=pelvis}
    componentAppliedLinearAccClamp=<1500 1500 1500>
    componentLinearAccScale=<0.2 0.1 0.1>
    componentLinearVelScale=<0.2 0.1 0.1>
    simSpaceSettings {
     maxAngularAcceleration=1000
     maxAngularVelocity=3
     maxLinearAcceleration=700
     maxLinearVelocity=500
     velocityScaleZ=0
     worldAlpha=0.5
     externalLinearDragV=<0.2 0.2>
    }
   }
   Save cached pose 'CRCache_RigidBody'(Pose=Component To Local(ComponentPose=@n41)) {cacheName=CRCache_RigidBody}
   n42 = Look At - Bone: base_hipAnkle_r(ComponentPose=Local To Component(LocalPose=Use cached pose 'CRCache_belt'()), Alpha=1) {
    alphaBoolBlend {blendOption=Linear}
    boneToModify {boneName=base_hipAnkle_r}
    lookAtLocation=<-4.4067 -14.4611 -4.5239>
    lookAtTarget {boneReference {boneName=calf_r}}
    lookAt_Axis {axis=<1 0>}
   }
   n43 = "Transform (Modify) Bone - Bone: base_hipAnkle_r"(ComponentPose=@n42, Alpha=Get crouching(), Translation="0,-1,0", Rotation="15.616,-25.320,-25.845", Scale="1,1,1") {
    rotationMode=BMM_Additive
    rotationSpace=BCS_BoneSpace
    translationMode=BMM_Additive
    translationSpace=BCS_BoneSpace
    alphaBoolBlend {blendOption=Linear}
    alphaScaleBiasClamp {bInterpResult=true}
    boneToModify {boneName=base_hipAnkle_r}
    rotation=<rot 15.616 -25.32 -25.845>
    translation {y=-1}
   }
   n44 = "Transform (Modify) Bone - Bone: base_hipAnkle_r"(ComponentPose=@n43, Alpha=Get OnlyJumping() -> (OnlyJumping:double), Translation="0,0,0", Rotation="24.237,-23.823,-26.864", Scale="1,1,1") {
    rotationMode=BMM_Additive
    rotationSpace=BCS_BoneSpace
    translationMode=BMM_Additive
    translationSpace=BCS_BoneSpace
    alphaBoolBlend {blendOption=Linear}
    alphaScaleBiasClamp {bInterpResult=true}
    boneToModify {boneName=base_hipAnkle_r}
    rotation=<rot 24.237 -23.823 -26.864>
   }
   Save cached pose 'CRCache_HipAnkle'(Pose=Component To Local(ComponentPose=@n44)) {cacheName=CRCache_HipAnkle}
   n45 = "Transform (Modify) Bone - Bone: base_bag_l"(
    ComponentPose=Local To Component(LocalPose=Use cached pose 'CRCache_Cup'())
    Alpha=1
    Translation="0,0,0"
    Rotation="8.943,-5.606,0.002"
    Scale="1,1,1"
   ) {
    rotationMode=BMM_Additive
    rotationSpace=BCS_BoneSpace
    translationMode=BMM_Additive
    translationSpace=BCS_BoneSpace
    alphaBoolBlend {blendOption=Linear}
    alphaScaleBiasClamp {bInterpResult=true}
    boneToModify {boneName=base_bag_l}
    rotation=<rot 8.943 -5.606 0.002>
   }
   n46 = "Transform (Modify) Bone - Bone: base_belt_c"(ComponentPose=@n45, Alpha=Get sprinting(), Translation="0,-1,0", Rotation="0,0,0", Scale="1,1,1") {
    rotationMode=BMM_Additive
    rotationSpace=BCS_BoneSpace
    translationMode=BMM_Additive
    translationSpace=BCS_BoneSpace
    alphaBoolBlend {blendOption=Linear}
    alphaScaleBiasClamp {bInterpResult=true}
    boneToModify {boneName=base_belt_c}
    translation {y=-1}
   }
   n47 = "Transform (Modify) Bone - Bone: base_bag_l"(ComponentPose=@n46, Alpha=Get Moving2D() -> (Moving2D:double), Translation="0.8,0,-0.003", Rotation="10.587,-0.000,-9.308", Scale="1,1,1") {
    rotationMode=BMM_Additive
    rotationSpace=BCS_BoneSpace
    translationMode=BMM_Additive
    translationSpace=BCS_BoneSpace
    alphaBoolBlend {blendOption=Linear}
    alphaScaleBiasClamp {bInterpResult=true}
    boneToModify {boneName=base_bag_l}
    rotation=<rot 10.587 0 -9.3081>
    translation {x=0.8, z=-0.003}
   }
   n48 = "Transform (Modify) Bone - Bone: base_bag_l"(ComponentPose=@n47, Alpha=Get InSky() -> (InSky:double), Translation="0,-1,0", Rotation="-25.925,-2.831,-4.650", Scale="1,1,1") {
    rotationMode=BMM_Additive
    rotationSpace=BCS_BoneSpace
    translationMode=BMM_Additive
    translationSpace=BCS_BoneSpace
    alphaBoolBlend {blendOption=Linear}
    alphaScaleBiasClamp {bInterpResult=true}
    boneToModify {boneName=base_bag_l}
    rotation=<rot -25.925 -2.831 -4.65>
    translation {y=-1}
   }
   n49 = Get crouching()
   n50 = "Transform (Modify) Bone - Bone: base_bag_l"(ComponentPose=@n48, Alpha=@n49, Translation="0.309,-1.491,3.122", Rotation="12.462,-18.980,6.384", Scale="1,1,1") {
    rotationMode=BMM_Additive
    rotationSpace=BCS_BoneSpace
    translationMode=BMM_Additive
    translationSpace=BCS_BoneSpace
    alphaBoolBlend {blendOption=Linear}
    alphaScaleBiasClamp {bInterpResult=true}
    boneToModify {boneName=base_bag_l}
    rotation=<rot 12.462 -18.98 6.384>
    translation=<0.309 -1.491 3.122>
   }
   n51 = "Transform (Modify) Bone - Bone: base_belt_c"(ComponentPose=@n50, Alpha=@n49, Translation="-1.267,-5.153,-0.590", Rotation="-0.006,-0.011,16.996", Scale="1,1,1") {
    rotationMode=BMM_Additive
    rotationSpace=BCS_BoneSpace
    translationMode=BMM_Additive
    translationSpace=BCS_BoneSpace
    alphaBoolBlend {blendOption=Linear}
    alphaScaleBiasClamp {bInterpResult=true}
    boneToModify {boneName=base_belt_c}
    rotation=<rot -0.006 -0.011 16.9959>
    translation=<-1.267 -5.153 -0.59>
   }
   n52 = "Transform (Modify) Bone - Bone: base_bag_l"(ComponentPose=@n51, Alpha=Get OnlyJumping(), Translation="0.146,-0.913,0.016", Rotation="7.891,-11.017,8.095", Scale="1,1,1") {
    rotationMode=BMM_Additive
    rotationSpace=BCS_BoneSpace
    translationMode=BMM_Additive
    translationSpace=BCS_BoneSpace
    alphaBoolBlend {blendOption=Linear}
    alphaScaleBiasClamp {bInterpResult=true}
    boneToModify {boneName=base_bag_l}
    rotation=<rot 7.891 -11.017 8.0949>
    translation=<0.146 -0.913 0.016>
   }
   n53 = "Transform (Modify) Bone - Bone: base_belt_c"(ComponentPose=@n52, Alpha=1, Translation="0,0,0", Rotation="0,0,0", Scale="1,1,1") {
    rotationMode=BMM_Additive
    rotationSpace=BCS_BoneSpace
    translationMode=BMM_Additive
    translationSpace=BCS_BoneSpace
    alphaBoolBlend {blendOption=Linear}
    alphaScaleBiasClamp {bInterpResult=true}
    boneToModify {boneName=base_belt_c}
   }
   n54 = "Transform (Modify) Bone - Bone: base_bag_l"(ComponentPose=@n53, Alpha=Get targetting() -> (targetting:double), Translation="0,0,0", Rotation="7.452,-4.853,-8.313", Scale="1,1,1") {
    rotationMode=BMM_Additive
    rotationSpace=BCS_BoneSpace
    translationMode=BMM_Additive
    translationSpace=BCS_BoneSpace
    alphaBoolBlend {blendOption=Linear}
    alphaScaleBiasClamp {bInterpResult=true}
    boneToModify {boneName=base_bag_l}
    rotation=<rot 7.452 -4.853 -8.313>
   }
   n55 = Get crauchingMove() -> (crauchingMove:double)
   n56 = "Transform (Modify) Bone - Bone: base_belt_c"(ComponentPose=@n54, Alpha=@n55, Translation="-0.011,-6.400,0.565", Rotation="0.015,0.033,27.581", Scale="1,1,1") {
    rotationMode=BMM_Additive
    rotationSpace=BCS_BoneSpace
    translationMode=BMM_Additive
    translationSpace=BCS_BoneSpace
    alphaBoolBlend {blendOption=Linear}
    alphaScaleBiasClamp {bInterpResult=true}
    boneToModify {boneName=base_belt_c}
    rotation=<rot 0.015 0.033 27.581>
    translation=<-0.011 -6.4 0.565>
   }
   n57 = "Transform (Modify) Bone - Bone: base_bag_l"(ComponentPose=@n56, Alpha=@n55, Translation="-0.02,-0.079,1.808", Rotation="-23.926,-3.200,-0.464", Scale="1,1,1") {
    rotationMode=BMM_Additive
    rotationSpace=BCS_BoneSpace
    translationMode=BMM_Additive
    translationSpace=BCS_BoneSpace
    alphaBoolBlend {blendOption=Linear}
    alphaScaleBiasClamp {bInterpResult=true}
    boneToModify {boneName=base_bag_l}
    rotation=<rot -23.926 -3.2 -0.464>
    translation=<-0.02 -0.079 1.808>
   }
   Save cached pose 'CRCache_belt'(Pose=Component To Local(ComponentPose=@n57)) {cacheName=CRCache_belt}
   n58 = Get targetting()
   n59 = "Transform (Modify) Bone - Bone: base_collar2A_l"(
    ComponentPose=Local To Component(LocalPose=Use cached pose 'CRCache_HipAnkle'())
    Alpha=@n58
    Translation="0.000,0.917,-0.207"
    Rotation="0,0,0"
    Scale="1,1,1"
   ) {
    rotationMode=BMM_Additive
    rotationSpace=BCS_BoneSpace
    translationMode=BMM_Additive
    translationSpace=BCS_BoneSpace
    alphaBoolBlend {blendOption=Linear}
    alphaScaleBiasClamp {bInterpResult=true}
    boneToModify {boneName=base_collar2A_l}
    translation {y=0.917, z=-0.207}
   }
   n60 = "Transform (Modify) Bone - Bone: base_collar1A_l"(ComponentPose=@n59, Alpha=@n58, Translation="-0.000,0.678,0.291", Rotation="0,0,0", Scale="1,1,1") {
    rotationMode=BMM_Additive
    rotationSpace=BCS_BoneSpace
    translationMode=BMM_Additive
    translationSpace=BCS_BoneSpace
    alphaBoolBlend {blendOption=Linear}
    alphaScaleBiasClamp {bInterpResult=true}
    boneToModify {boneName=base_collar1A_l}
    translation {y=0.678, z=0.291}
   }
   n61 = "Transform (Modify) Bone - Bone: base_collarA_c"(
    ComponentPose=@n60
    Alpha=Get IsLookingUp() -> (IsLookingUp:double)
    Translation="-0.124,1.197,-0.051"
    Rotation="49.843,-0.008,-0.006"
    Scale="1,1,1"
   ) {
    rotationMode=BMM_Additive
    rotationSpace=BCS_BoneSpace
    translationMode=BMM_Additive
    translationSpace=BCS_BoneSpace
    alphaBoolBlend {blendOption=Linear}
    alphaScaleBiasClamp {bInterpResult=true, interpSpeedDecreasing=5, interpSpeedIncreasing=8}
    boneToModify {boneName=base_collarA_c}
    rotation=<rot 49.843 -0.008 -0.006>
    translation=<-0.124 1.197 -0.051>
   }
   n62 = "Transform (Modify) Bone - Bone: base_collarA_c"(ComponentPose=@n61, Alpha=Get InSky(), Translation="-0.124,1.197,-0.051", Rotation="35,-0.008,-0.006", Scale="1,1,1") {
    rotationMode=BMM_Additive
    rotationSpace=BCS_BoneSpace
    translationMode=BMM_Additive
    translationSpace=BCS_BoneSpace
    alphaBoolBlend {blendOption=Linear}
    alphaScaleBiasClamp {bInterpResult=true}
    boneToModify {boneName=base_collarA_c}
    rotation=<rot 35 -0.008 -0.006>
    translation=<-0.124 1.197 -0.051>
   }
   Save cached pose 'CRCache_Neck'(Pose=Component To Local(ComponentPose=@n62)) {cacheName=CRCache_Neck}
  ]}
 EventGraph {nodes [
   Event BlueprintUpdateAnimation()
   > n1 = Sequence!(execute) -> (then_0, then_1, then_2:Exec)
   > n2 = Sequence()
   n3 = Branch(execute=@n2.then_2, Condition=AND Boolean(A=Get bIsCrouching(), B=Get bIsMoving2D()))
   > Set CrouchMove_alpha(CrouchMove_alpha:double=0.5) -> (Output_Get:double)
   Set CrouchMove_alpha(execute=@n3.else, CrouchMove_alpha=1)
   Set crouching(execute=@n2.then_1, crouching:double="To Float (Boolean)"(InBool=Get bIsCrouching())) -> (Output_Get:double)
   n4 = Get bIsFalling()
   n5 = Get bIsJumping()
   > Set OnlyJumping(OnlyJumping:double="To Float (Boolean)"(InBool=OR Boolean(A=@n5, B=@n4))) -> (Output_Get:double)
   > Set Moving2D(Moving2D:double="To Float (Boolean)"(InBool=Get bIsMoving2D())) -> (Output_Get:double)
   n6 = OR Boolean(A=Get bIsJumping(), B=Get bIsFalling())
   n7 = OR Boolean!(A=Get bIsParachuteOpened(), B=Get bIsSkydiveFloating(), C:bool=Get bIsSkydiveDiveMode()) -> (ReturnValue)
   > Set jumping(jumping:double="To Float (Boolean)"(InBool=NAND Boolean(A=@n6, B=@n7) {is=title K2Node_CommutativeAssociativeBinaryOperator})) -> (Output_Get:double)
   n8 = Get bIsSprinting()
   n9 = Get bIsTacticalSprinting()
   > Set sprinting(sprinting:double="To Float (Boolean)"(InBool=OR Boolean!(A=@n9, B=@n8) -> (ReturnValue))) -> (Output_Get:double)
   n10 = Get bIsTargeting()
   n11 = Get bIsTargetingConsumableThrow() -> (bIsTargetingConsumableThrow:bool)
   > Set targetting(targetting:double="To Float (Boolean)"(InBool=OR Boolean(A=@n10, B=@n11))) -> (Output_Get:double)
   n12 = Get bIsCrouching()
   n13 = Get bIsMoving2D()
   > Set crauchingMove(crauchingMove:double="To Float (Boolean)"(InBool=AND Boolean(A=@n12, B=@n13))) -> (Output_Get:double)
   n14 = Get bIsSkyDiving()
   n15 = Get bIsSkydiveDiveMode()
   n16 = Get bIsSkydiveFloating()
   > Set InSky(InSky:double="To Float (Boolean)"(InBool=OR Boolean!(A=@n14, B=@n16, C:bool=@n15) -> (ReturnValue))) -> (Output_Get:double)
   n17 = OR Boolean!(A=Get bIsTacticalSprinting(), B=Get bIsSprinting()) -> (ReturnValue)
   n18 = Get bIsJumping()
   > Set jumpingSprinting(jumpingSprinting:double="To Float (Boolean)"(InBool=AND Boolean(A=@n17, B=@n18))) -> (Output_Get:double)
   > Set KneeslidingAlpha(KneeslidingAlpha:float="To Float (Boolean)"(InBool=Get bIsKneeSliding())) -> (Output_Get:float)
   n19 = OR Boolean(A=Get bIsTargeting(), B=Get bIsTargetingConsumableThrow())
   n20 = GetOwnerSkeletalMeshComponent()
   n21 = BreakRotator(InRot=BreakTransform(InTransform=GetDeltaTransformFromRefPose(self=@n20, BoneName=neck_02, BaseName=head)).Rotation)
   n22 = "float >= float"(A=@n21.Yaw, B=2)
   > Set TargetingUp(TargetingUp:double="To Float (Boolean)"(InBool=AND Boolean(A=@n19, B=@n22))) -> (Output_Get:double)
   n23 = Gravity Override State Function DEPRECATED(
    execute=@n2.then_0
    CharacterPartAnimInstance=Self-Reference()
    GameTimeInSeconds=GetGameTimeInSeconds()
    GravityOverrideParams=Get Gravity Override Params()
   )
   > Set Override World Gravity(Override World Gravity:Vector=@n23) -> (Output_Get:Vector)
   n24 = Sequence(execute=@n1.then_1)
   n25 = PlayerMovement(execute=@n24.then_2, AnimInstance=Self-Reference(), XDotProductCoefficient=100, YDotProductCoefficient=100, ZDotProductCoefficient=100)
   n26 = BreakVector(InVec=MakeVector(X=@n25, Y=@n25.YMovement))
   n27 = "float < float"(A=@n26, B=-50)
   n28 = Get bIsCrouching()
   > Set Left(Left:double="To Float (Boolean)"(InBool=AND Boolean!(A=@n27, B=@n28, C:bool) -> (ReturnValue))) -> (Output_Get:double)
   n29 = NOT Boolean(A=OR Boolean(A=Get bIsJumping(), B=Get bIsFalling()))
   n30 = "float > float"(A=@n26, B=50)
   > Set Right(Right:double="To Float (Boolean)"(InBool=AND Boolean(A=@n30, B=@n29, C=@n28))) -> (Output_Get:double)
   > Set Forward(Forward:double="To Float (Boolean)"(InBool="float > float"(A=@n26.Y, B=40))) -> (Output_Get:double)
   > Set Backward(Backward:double="To Float (Boolean)"(InBool="float < float"(A=@n26.Y, B=-40))) -> (Output_Get:double)
   > Set WalkRight(WalkRight:double="To Float (Boolean)"(InBool=@n29)) -> (Output_Get:double)
   > Set walkLeft(walkLeft:double="To Float (Boolean)"(InBool=@n27)) -> (Output_Get:double)
   n31 = BreakTransform(InTransform=GetDeltaTransformFromRefPose(self=GetOwnerSkeletalMeshComponent(), BoneName=neck_02, BaseName=head))
   n32 = Branch(execute=@n24.then_1, Condition="float <= float"(A=BreakRotator(InRot=@n31.Rotation).Yaw, B=-1) -> (ReturnValue=~))
   > Set IsLookingDown(IsLookingDown:double=1) -> (Output_Get:double)
   Set IsLookingDown(execute=@n32.else, IsLookingDown=0)
   n33 = Branch(execute=@n24.then_0, Condition="float >= float"(A=@n21.Yaw, B=20))
   > Set IsLookingUp(IsLookingUp:double=1) -> (Output_Get:double)
   Set IsLookingUp(execute=@n33.else, IsLookingUp=0)
  ]}
]
variables [
 Backward {type=number}
 crauchingMove {type=number}
 crouching {type=number}
 CrouchMove_alpha {type=number}
 Forward {type=number}
 Gravity Override Params {type=GravityOverrideParamsStruct}
 InSky {type=number}
 IsLookingDown {type=number}
 IsLookingUp {type=number}
 jumping {type=number}
 jumpingSprinting {type=number}
 KneeslidingAlpha {type=number, default=1}
 Left {type=number}
 Moving2D {type=number}
 OnlyJumping {type=number}
 Override World Gravity {type=Vector, default=<0 0 -980>}
 Right {type=number}
 sprinting {type=number}
 TargetingUp {type=number}
 targetting {type=number}
 walkLeft {type=number}
 WalkRight {type=number}
]
