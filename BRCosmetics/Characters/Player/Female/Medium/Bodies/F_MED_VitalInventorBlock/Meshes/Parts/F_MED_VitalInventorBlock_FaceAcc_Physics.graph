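# Physics asset for the FaceAcc mesh part of the female medium VitalInventorBlock player body, covering hair, ponytail and collar.
# Capsule bodies on head, neck, upper spine, right clavicle and upper arm accompany those on the dynamic bones.
# Front hair forms left and right chains from the head, and the ponytail branches into side strands, all with tight swing limits mostly.
# Collar, ponytail base and most torso joints leave swing and twist free.
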PhysicsAsset /BRCosmetics/Characters/Player/Female/Medium/Bodies/F_MED_VitalInventorBlock/Meshes/Parts/F_MED_VitalInventorBlock_FaceAcc_Physics
bodies {
 base_head [{shapeType=Capsule, length=5.3985, radius=6.6015, shapeName=base_head_capsule, center=<1 0 0>, rotation=<rot 90 0 0>}]
 base_ponyTail1 [{shapeType=Capsule, length=0.505, radius=0.505, shapeName=base_ponyTail1_capsule, rotation=<rot 0 -90 0>}]
 clavicle_r [{shapeType=Capsule, length=5.5454, radius=9.3656, shapeName=clavicle_r_capsule, center=<-9.7471 -1.0835 1.4561>, rotation=<rot 90 0 0>}]
 dyn_collar1A_l [
  {shapeType=Capsule, length=7.6166, radius=1.505, shapeName=dyn_collar1A_l_capsule, center=<3.4736 0 -1.6081>}
  {shapeType=Capsule, length=7, radius=1.505, shapeName=dyn_collar1A_l_capsule, center=<1.3257 0.3687 -1.2295>, rotation=<rot 0.0001 0 -5.585>}
 ]
 dyn_collar1A_r [
  {shapeType=Capsule, length=7, radius=1, shapeName=dyn_collar1A_l_capsule, center=<4.4706 -0.6582 -1.2398>}
  {shapeType=Capsule, length=7, radius=2, shapeName=dyn_collar1A_l_capsule, center=<1.5 0 -1.2398>}
 ]
 dyn_collar2A_l [
  {shapeType=Capsule, length=2.5758, radius=1.505, shapeName=dyn_collar1A_l_capsule}
  {shapeType=Capsule, length=2.5758, radius=1.505, shapeName=dyn_collar1A_l_capsule, center=<2.2714 0 0>}
 ]
 dyn_hairFront01_l [{shapeType=Capsule, length=1.7726, radius=1.2688, shapeName=dyn_hairFront01_l_capsule, center=<1 0 0>, rotation=<rot 90 0 0>}]
 dyn_hairFront01_r [{shapeType=Capsule, length=1.6747, radius=1.2013, shapeName=dyn_hairFront01_r_capsule, center=<1 0 0>, rotation=<rot 90 0 0>}]
 dyn_hairFront02_l [{shapeType=Capsule, length=1.869, radius=1.2094, shapeName=dyn_hairFront02_l_capsule, center=<1 0 0>, rotation=<rot 90 0 0>}]
 dyn_hairFront02_r [{shapeType=Capsule, length=2.0817, radius=1.5167, shapeName=dyn_hairFront02_r_capsule, center=<1 0 0>, rotation=<rot 90 0 0>}]
 dyn_hairFront04_l [{
   shapeType=Capsule
   length=4.2986
   radius=0.5
   shapeName=dyn_hairFront04_l_capsule
   center=<1.2605 0.7627 -0.4328>
   rotation=<rot 71.5164 0.0002 -23.0252>
  }]
 dyn_hairFront04_r [{
   shapeType=Capsule
   length=3.9813
   radius=0.6
   shapeName=dyn_hairFront04_r_capsule
   center=<1.6935 1.0536 0.7538>
   rotation=<rot 75.0566 -152.2753 -180>
  }]
 dyn_ponyTail [{shapeType=Capsule, length=1.4183, radius=2.1775, shapeName=dyn_ponyTail_capsule, center=<3.0284 0 0>, rotation=<rot 90 0 0>}]
 dyn_ponyTail01 [{shapeType=Capsule, length=0.8092, radius=3.3411, shapeName=dyn_ponyTail01_capsule, center=<3.0473 0 0>, rotation=<rot 90 0 0>}]
 dyn_ponyTail02 [{shapeType=Capsule, length=2.2428, radius=2.4444, shapeName=dyn_ponyTail02_capsule, center=<2 0 0>, rotation=<rot 90 0 0>}]
 dyn_ponyTailHairB02_r [{
   shapeType=Capsule
   length=1.8713
   radius=2.3821
   shapeName=dyn_ponyTailHairB02_r_capsule
   center=<3.6337 0.527 -0.6197>
   rotation=<rot 90 180 157.2592>
  }]
 dyn_ponyTailHairC01_l [{shapeType=Capsule, length=3.243, radius=1.4588, shapeName=dyn_ponyTailHairC01_l_capsule, center=<2.5969 0 0>, rotation=<rot 90 0 0>}]
 dyn_ponyTailHairF01_l [{shapeType=Capsule, length=3.8452, radius=3.5303, shapeName=dyn_ponyTailHairF01_l_capsule, center=<2.8311 0 0>, rotation=<rot 90 0 0>}]
 neck_02 [{shapeType=Capsule, length=10.4101, radius=5.5899, shapeName=neck_02_capsule, center=<-0.373 -1.7237 0>, rotation=<rot 0 167.79 -180>}]
 spine_05 [{shapeType=Capsule, length=17.99, radius=10.51, shapeName=spine_05_capsule, center=<2.4194 1.4118 0.018>, rotation=<rot 0.6771 179.9999 180>}]
 upperarm_r [{shapeType=Capsule, length=23.5675, radius=3.6163, shapeName=upperarm_r_capsule, center=<-14.5393 1.0324 0>, rotation=<rot 90 0 0>}]
}
constraints [
 {bone1Name=base_head, bone2Name=neck_02, swing1LimitDegrees=1, swing2LimitDegrees=1, twistLimitDegrees=0}
 {
  bone1Name=neck_02
  bone2Name=spine_05
  swing1LimitDegrees=45
  swing1Motion=Free
  swing2LimitDegrees=45
  swing2Motion=Free
  twistLimitDegrees=45
  twistMotion=Free
 }
 {
  bone1Name=base_ponyTail1
  bone2Name=base_head
  swing1LimitDegrees=45
  swing1Motion=Free
  swing2LimitDegrees=45
  swing2Motion=Free
  twistLimitDegrees=45
  twistMotion=Free
 }
 {bone1Name=dyn_ponyTail, bone2Name=base_ponyTail1, swing1LimitDegrees=7, swing2LimitDegrees=12, twistLimitDegrees=0}
 {bone1Name=dyn_ponyTail01, bone2Name=dyn_ponyTail, swing1LimitDegrees=45, swing2LimitDegrees=45, twistLimitDegrees=0}
 {bone1Name=dyn_ponyTail02, bone2Name=dyn_ponyTail01, swing1LimitDegrees=25, swing2LimitDegrees=25, twistLimitDegrees=0}
 {bone1Name=dyn_ponyTailHairF01_l, bone2Name=dyn_ponyTail02, swing1LimitDegrees=15, swing2LimitDegrees=15, twistLimitDegrees=0}
 {bone1Name=dyn_ponyTailHairC01_l, bone2Name=dyn_ponyTail01, swing1LimitDegrees=15, swing2LimitDegrees=15, twistLimitDegrees=0}
 {bone1Name=dyn_hairFront01_l, bone2Name=base_head, swing1LimitDegrees=20, swing2LimitDegrees=15, twistLimitDegrees=0}
 {bone1Name=dyn_hairFront02_l, bone2Name=dyn_hairFront01_l, swing1LimitDegrees=25, swing2LimitDegrees=15, twistLimitDegrees=0}
 {bone1Name=dyn_hairFront04_l, bone2Name=dyn_hairFront02_l, swing1LimitDegrees=25, swing2LimitDegrees=15, twistLimitDegrees=0}
 {bone1Name=dyn_hairFront01_r, bone2Name=base_head, swing1LimitDegrees=20, swing2LimitDegrees=15, twistLimitDegrees=0}
 {bone1Name=dyn_hairFront02_r, bone2Name=dyn_hairFront01_r, swing1LimitDegrees=25, swing2LimitDegrees=15, twistLimitDegrees=0}
 {bone1Name=dyn_hairFront04_r, bone2Name=dyn_hairFront02_r, swing1LimitDegrees=25, swing2LimitDegrees=15, twistLimitDegrees=0}
 {
  bone1Name=upperarm_r
  bone2Name=clavicle_r
  swing1LimitDegrees=45
  swing1Motion=Free
  swing2LimitDegrees=45
  swing2Motion=Free
  twistLimitDegrees=45
  twistMotion=Free
 }
 {bone1Name=dyn_ponyTailHairB02_r, bone2Name=dyn_ponyTail02, swing1LimitDegrees=10, swing2LimitDegrees=10, twistLimitDegrees=0}
 {
  bone1Name=clavicle_r
  bone2Name=spine_05
  swing1LimitDegrees=45
  swing1Motion=Free
  swing2LimitDegrees=45
  swing2Motion=Free
  twistLimitDegrees=45
  twistMotion=Free
 }
 {
  bone1Name=dyn_collar1A_l
  bone2Name=spine_05
  swing1LimitDegrees=45
  swing1Motion=Free
  swing2LimitDegrees=45
  swing2Motion=Free
  twistLimitDegrees=45
  twistMotion=Free
 }
 {
  bone1Name=dyn_collar2A_l
  bone2Name=spine_05
  swing1LimitDegrees=45
  swing1Motion=Free
  swing2LimitDegrees=45
  swing2Motion=Free
  twistLimitDegrees=45
  twistMotion=Free
 }
 {
  bone1Name=dyn_collar1A_r
  bone2Name=spine_05
  swing1LimitDegrees=45
  swing1Motion=Free
  swing2LimitDegrees=45
  swing2Motion=Free
  twistLimitDegrees=45
  twistMotion=Free
 }
 {
  bone1Name=dyn_collar2A_l
  bone2Name=neck_02
  swing1LimitDegrees=45
  swing1Motion=Free
  swing2LimitDegrees=45
  swing2Motion=Free
  twistLimitDegrees=45
  twistMotion=Free
 }
]
