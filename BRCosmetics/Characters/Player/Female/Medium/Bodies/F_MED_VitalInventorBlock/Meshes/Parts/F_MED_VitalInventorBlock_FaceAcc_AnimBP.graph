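# Animation Blueprint that adds secondary hair motion to the VitalInventorBlock female face accessory part.
# It copies the attached parent's pose, then Look At nodes aim the ponytail and front hair base bones downward.
# A spine_05-space RigidBody simulation follows, then six chained, spring-driven AnimDynamics nodes on ponytail bones output the pose.
# The EventGraph updates RigidBody gravity via a deprecated override function and sets inSkyDive from skydive and parachute states.
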AnimBlueprint /BRCosmetics/Characters/Player/Female/Medium/Bodies/F_MED_VitalInventorBlock/Meshes/Parts/F_MED_VitalInventorBlock_FaceAcc_AnimBP
target_skeleton=&/F_MED_VitalInventorBlock_FaceAcc_Skeleton
graphs [
 AnimGraph {nodes [
   n1 = Anim Dynamics - Bone: dyn_ponyTailHairE01_l!(
    ComponentPose=Local To Component(LocalPose=Use cached pose 'PhysicsAsset Hairs'())
    Alpha=1
    GravityOverride:Vector="0,0,-980"
    ExternalForce="0,0,0"
   ) -> (Pose) {
    angularDampingOverride=1
    angularSpringConstant=3000
    bAngularSpring=true
    bOverrideAngularDamping=true
    bUseGravityOverride=true
    gravityScale=0.35
    simulationSpace=BoneRelative
    alphaBoolBlend {blendOption=Linear}
    boundBone {boneName=dyn_ponyTailHairE01_l}
    componentAppliedLinearAccClamp=<3000 3000 3000>
    componentLinearAccScale=<0.3 0.2 0.2>
    componentLinearVelScale=<0.3 0.2 0.2>
    gravityOverride {z=-980}
    physicsBodyDefinitions [{
      boundBone {boneName=dyn_ponyTailHairE01_l}
      boxExtents=<10 10 10>
      constraintSetup {angularLimitsMax=<0 15 15>, angularLimitsMin=<0 -15 -15>, angularTarget=<1 0 0>}
      localJointOffset=<3 0 0>
     }]
    relativeSpaceBone {boneName=neck_01}
   }
   n2 = Anim Dynamics - Bone: dyn_ponyTailHairD01_l!(ComponentPose=@n1, Alpha=1, GravityOverride:Vector="0,0,-980", ExternalForce="0,0,0") -> (Pose) {
    angularDampingOverride=1
    angularSpringConstant=3000
    bAngularSpring=true
    bOverrideAngularDamping=true
    bUseGravityOverride=true
    gravityScale=0.35
    simulationSpace=BoneRelative
    alphaBoolBlend {blendOption=Linear}
    boundBone {boneName=dyn_ponyTailHairD01_l}
    componentAppliedLinearAccClamp=<3000 3000 3000>
    componentLinearAccScale=<0.3 0.2 0.2>
    componentLinearVelScale=<0.3 0.2 0.2>
    gravityOverride {z=-980}
    physicsBodyDefinitions [{
      boundBone {boneName=dyn_ponyTailHairD01_l}
      boxExtents=<10 10 10>
      constraintSetup {angularLimitsMax=<0 15 15>, angularLimitsMin=<0 -15 -15>, angularTarget=<1 0 0>}
      localJointOffset=<3 0 0>
     }]
    relativeSpaceBone {boneName=neck_01}
   }
   n3 = Anim Dynamics - Bone: dyn_ponyTailHairA01_r!(ComponentPose=@n2, Alpha=1, GravityOverride:Vector="0,0,-980", ExternalForce="0,0,0") -> (Pose) {
    angularDampingOverride=1
    angularSpringConstant=3000
    bAngularSpring=true
    bOverrideAngularDamping=true
    bUseGravityOverride=true
    gravityScale=0.35
    simulationSpace=BoneRelative
    alphaBoolBlend {blendOption=Linear}
    boundBone {boneName=dyn_ponyTailHairA01_r}
    componentAppliedLinearAccClamp=<3000 3000 3000>
    componentLinearAccScale=<0.3 0.2 0.2>
    componentLinearVelScale=<0.3 0.2 0.2>
    gravityOverride {z=-980}
    physicsBodyDefinitions [{
      boundBone {boneName=dyn_ponyTailHairA01_r}
      boxExtents=<10 10 10>
      constraintSetup {angularLimitsMax=<0 15 15>, angularLimitsMin=<0 -15 -15>, angularTarget=<1 0 0>}
      localJointOffset=<3 0 0>
     }]
    relativeSpaceBone {boneName=neck_01}
   }
   n4 = Anim Dynamics - Bone: dyn_ponyTailHairA01_c!(ComponentPose=@n3, Alpha=1, GravityOverride:Vector="0,0,-980", ExternalForce="0,0,0") -> (Pose) {
    angularDampingOverride=1
    angularSpringConstant=3000
    bAngularSpring=true
    bOverrideAngularDamping=true
    bUseGravityOverride=true
    gravityScale=0.35
    simulationSpace=BoneRelative
    alphaBoolBlend {blendOption=Linear}
    boundBone {boneName=dyn_ponyTailHairA01_c}
    componentAppliedLinearAccClamp=<3000 3000 3000>
    componentLinearAccScale=<0.3 0.2 0.2>
    componentLinearVelScale=<0.3 0.2 0.2>
    gravityOverride {z=-980}
    physicsBodyDefinitions [{
      boundBone {boneName=dyn_ponyTailHairA01_c}
      boxExtents=<10 10 10>
      constraintSetup {angularLimitsMax=<0 15 15>, angularLimitsMin=<0 -15 -15>, angularTarget=<1 0 0>}
      localJointOffset=<3 0 0>
     }]
    relativeSpaceBone {boneName=neck_01}
   }
   n5 = Anim Dynamics - Bone: dyn_ponyTailHairB01_l!(ComponentPose=@n4, Alpha=1, GravityOverride:Vector="0,0,-980", ExternalForce="0,0,0") -> (Pose) {
    angularDampingOverride=1
    angularSpringConstant=3000
    bAngularSpring=true
    bOverrideAngularDamping=true
    bUseGravityOverride=true
    gravityScale=0.35
    simulationSpace=BoneRelative
    alphaBoolBlend {blendOption=Linear}
    boundBone {boneName=dyn_ponyTailHairB01_l}
    componentAppliedLinearAccClamp=<3000 3000 3000>
    componentLinearAccScale=<0.3 0.2 0.2>
    componentLinearVelScale=<0.3 0.2 0.2>
    gravityOverride {z=-980}
    physicsBodyDefinitions [{
      boundBone {boneName=dyn_ponyTailHairB01_l}
      boxExtents=<10 10 10>
      constraintSetup {angularLimitsMax=<0 15 15>, angularLimitsMin=<0 -15 -15>, angularTarget=<1 0 0>}
      localJointOffset=<3 0 0>
     }]
    relativeSpaceBone {boneName=neck_01}
   }
   n6 = Anim Dynamics - Bone: dyn_ponyTailHairA01_l!(ComponentPose=@n5, Alpha=1, GravityOverride:Vector="0,0,-980", ExternalForce="0,0,0") -> (Pose) {
    angularDampingOverride=1
    angularSpringConstant=3000
    bAngularSpring=true
    bOverrideAngularDamping=true
    bUseGravityOverride=true
    gravityScale=0.35
    simulationSpace=BoneRelative
    alphaBoolBlend {blendOption=Linear}
    boundBone {boneName=dyn_ponyTailHairA01_l}
    componentAppliedLinearAccClamp=<3000 3000 3000>
    componentLinearAccScale=<0.3 0.2 0.2>
    componentLinearVelScale=<0.3 0.2 0.2>
    gravityOverride {z=-980}
    physicsBodyDefinitions [{
      boundBone {boneName=dyn_ponyTailHairA01_l}
      boxExtents=<10 10 10>
      constraintSetup {angularLimitsMax=<0 10 10>, angularLimitsMin=<0 -10 -10>, angularTarget=<1 0 0>}
      localJointOffset=<3 0 0>
     }]
    relativeSpaceBone {boneName=neck_01}
   }
   Output Pose(Result=Component To Local(ComponentPose=@n6)) {layerGroup=None}
   n7 = Look At - Bone: base_ponyTail1(ComponentPose=Local To Component(LocalPose=Copy Pose From Mesh() {bUseAttachedParent=true}), Alpha=1) {
    alphaBoolBlend {blendOption=Linear}
    boneToModify {boneName=base_ponyTail1}
    lookAtLocation=<0 -8.4628 0>
    lookAtTarget {boneReference {boneName=spine_05}}
    lookAt_Axis {axis=<1 0>}
   }
   n8 = Look At - Bone: base_hairFront01_r(ComponentPose=@n7, Alpha=0.8) {
    alphaBoolBlend {blendOption=Linear}
    boneToModify {boneName=base_hairFront01_r}
    lookAtLocation {x=0}
    lookAtTarget {boneReference {boneName=root}}
    lookAt_Axis {axis=<1 0>}
   }
   n9 = Look At - Bone: base_hairFront01_l(ComponentPose=@n8, Alpha=0.8) {
    alphaBoolBlend {blendOption=Linear}
    boneToModify {boneName=base_hairFront01_l}
    lookAtLocation {x=0, z=0}
    lookAtTarget {boneReference {boneName=root}}
    lookAt_Axis {axis=<1 0>}
   }
   Save cached pose 'Hair Pointing down'(Pose=Component To Local(ComponentPose=@n9)) {cacheName=Hair Pointing down}
   n10 = RigidBody!(
    ComponentPose=Local To Component(LocalPose=Use cached pose 'Hair Pointing down'())
    Alpha
    OverrideWorldGravity:Vector=Get Override World Gravity() -> (Override World Gravity:Vector)
    ExternalForce
   ) -> (Pose) {
    bOverrideWorldGravity=true
    overridePhysicsAsset=&/F_MED_VitalInventorBlock_FaceAcc_Physics
    simulationSpace=BaseBoneSpace
    alphaBoolBlend {blendOption=Linear}
    alphaScaleBiasClamp {bInterpResult=true}
    baseBoneRef {boneName=spine_05}
    componentAppliedLinearAccClamp=<3000 3000 3000>
    componentLinearAccScale=<0.3 0.3 0.3>
    componentLinearVelScale=<0.3 0.3 0.3>
    simSpaceSettings {
     maxAngularAcceleration=1000
     maxAngularVelocity=3
     maxLinearAcceleration=700
     maxLinearVelocity=500
     velocityScaleZ=0
     worldAlpha=0.5
     externalLinearDragV=<0.2 0.2>
    }
   }
   Save cached pose 'PhysicsAsset Hairs'(Pose=Component To Local(ComponentPose=@n10)) {cacheName=PhysicsAsset Hairs}
  ]}
 EventGraph {nodes [
   Event BlueprintUpdateAnimation()
   > n1 = Sequence!(execute) -> (then_0, then_1, then_2:Exec)
   > n2 = Gravity Override State Function DEPRECATED(CharacterPartAnimInstance=Self-Reference(), GameTimeInSeconds=GetGameTimeInSeconds(), GravityOverrideParams=Get Gravity Override Params())
   > Set Override World Gravity(Override World Gravity:Vector=@n2) -> (Output_Get:Vector)
   n3 = Branch(
    execute=@n1.then_1
    Condition=OR Boolean!(A=Get bIsSkydiveDiveMode(), B=Get bIsSkyDiving(), C:bool=Get bIsSkydiveFloating(), D:bool=Get bIsParachuteOpened()) -> (ReturnValue)
   )
   > Set inSkyDive(inSkyDive:double=0.5) -> (Output_Get:double)
   Set inSkyDive(execute=@n3.else, inSkyDive=1)
  ]}
]
variables [
 Gravity Override Params {type=GravityOverrideParamsStruct}
 inSkyDive {type=number}
 IsFallingAlpha {type=number, default=1}
 isLookUp {type=number}
 Override World Gravity {type=Vector}
 targeting {type=number}
]
